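# Secondary-motion AnimBlueprint for the M_MED_DryEraseToro_Flint male body cosmetic, following the pose of the attached parent mesh.
# The copied pose feeds the body's Control Rig, then a weighted constraint drives the bracelet bone from left forearm, wrist and hand bones.
# Eight chained single-bone AnimDynamics nodes then add damped, spring-driven jiggle to left and right jeans bones in world space.
# It has no variables or event logic.
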AnimBlueprint /BRCosmetics/Characters/Player/Male/Medium/Bodies/M_MED_DryEraseToro_Flint/Meshes/M_MED_DryEraseToro_Flint_AnimBP
target_skeleton=&/M_MED_DryEraseToro_Flint_Skeleton
graphs [
 AnimGraph {nodes [
   n1 = Control Rig(Source=Copy Pose From Mesh() {bUseAttachedParent=true}) {
    alphaBoolBlend {blendOption=Linear}
    controlRigAssetReference {blueprintRigClass=&/M_MED_DryEraseToro_Flint_CtrlRig.M_MED_DryEraseToro_Flint_CtrlRig_C}
   }
   n2 = "Constraint - bracelet (PPPPPPPPPP)"!(
    ComponentPose=Local To Component(LocalPose=@n1)
    Alpha=1
    ConstraintWeights_0=0.0161
    ConstraintWeights_1:float=0.0978
    ConstraintWeights_2:float=0.005
    ConstraintWeights_3:float=0.0408
    ConstraintWeights_4:float=0.0031
    ConstraintWeights_5:float=0.045
    ConstraintWeights_6:float=0.099
    ConstraintWeights_7:float=0.0801
    ConstraintWeights_8:float=0.5087
    ConstraintWeights_9:float=0.1043
   ) -> (Pose) {
    alphaBoolBlend {blendOption=Linear}
    boneToModify {boneName=bracelet}
    constraintSetup [
     {targetBone {boneName=lowerarm_l}}
     {targetBone {boneName=hand_l}}
     {targetBone {boneName=index_metacarpal_l}}
     {targetBone {boneName=pinky_metacarpal_l}}
     {targetBone {boneName=ring_metacarpal_l}}
     {targetBone {boneName=thumb_01_l}}
     {targetBone {boneName=deform_wrist_up_l}}
     {targetBone {boneName=deform_wrist_dn_l}}
     {targetBone {boneName=lowerarm_twist_01_l}}
     {targetBone {boneName=lowerarm_twist_02_l}}
    ]
    constraintWeights [1,1,1,1,1,1,1,1,1,1]
   }
   n3 = Anim Dynamics - Bone: BASE_dyn_jeans_l_01(ComponentPose=@n2, Alpha=1, ExternalForce="0,0,0") {
    angularDampingOverride=0.99
    angularSpringConstant=300
    bAngularSpring=true
    bLinearSpring=true
    bOverrideAngularDamping=true
    bOverrideLinearDamping=true
    bUseGravityOverride=true
    linearDampingOverride=0.999
    linearSpringConstant=60
    simulationSpace=World
    alphaBoolBlend {blendOption=Linear}
    boundBone {boneName=BASE_dyn_jeans_l_01}
    componentAppliedLinearAccClamp=<6000 6000 6000>
    componentLinearAccScale=<0.4 0.4 0.4>
    componentLinearVelScale=<0.4 0.4 0.4>
    physicsBodyDefinitions [{
      boundBone {boneName=BASE_dyn_jeans_l_01}
      boxExtents=<5 5 5>
      constraintSetup {angularTargetAxis=AxisY, twistAxis=AxisZ, angularLimitsMax=<60 30 30>, angularLimitsMin=<-60 -30 -30>, angularTarget=<0 1 0>}
      localJointOffset=<0 0 0.2>
     }]
    relativeSpaceBone {boneName=thigh_r}
   }
   n4 = Anim Dynamics - Bone: BASE_dyn_jeans_l_02(ComponentPose=@n3, Alpha=1, ExternalForce="0,0,0") {
    angularDampingOverride=0.99
    angularSpringConstant=300
    bAngularSpring=true
    bLinearSpring=true
    bOverrideAngularDamping=true
    bOverrideLinearDamping=true
    bUseGravityOverride=true
    linearDampingOverride=0.999
    linearSpringConstant=60
    simulationSpace=World
    alphaBoolBlend {blendOption=Linear}
    boundBone {boneName=BASE_dyn_jeans_l_02}
    componentAppliedLinearAccClamp=<6000 6000 6000>
    componentLinearAccScale=<0.4 0.4 0.4>
    componentLinearVelScale=<0.4 0.4 0.4>
    physicsBodyDefinitions [{
      boundBone {boneName=BASE_dyn_jeans_l_02}
      boxExtents=<5 5 5>
      constraintSetup {angularTargetAxis=AxisY, twistAxis=AxisZ, angularLimitsMax=<60 30 30>, angularLimitsMin=<-60 -30 -30>, angularTarget=<0 1 0>}
      localJointOffset=<0 0 0.2>
     }]
    relativeSpaceBone {boneName=thigh_r}
   }
   n5 = Anim Dynamics - Bone: BASE_dyn_jeans_l_03(ComponentPose=@n4, Alpha=1, ExternalForce="0,0,0") {
    angularDampingOverride=0.99
    angularSpringConstant=300
    bAngularSpring=true
    bLinearSpring=true
    bOverrideAngularDamping=true
    bOverrideLinearDamping=true
    bUseGravityOverride=true
    linearDampingOverride=0.999
    linearSpringConstant=60
    simulationSpace=World
    alphaBoolBlend {blendOption=Linear}
    boundBone {boneName=BASE_dyn_jeans_l_03}
    componentAppliedLinearAccClamp=<6000 6000 6000>
    componentLinearAccScale=<0.4 0.4 0.4>
    componentLinearVelScale=<0.4 0.4 0.4>
    physicsBodyDefinitions [{
      boundBone {boneName=BASE_dyn_jeans_l_03}
      boxExtents=<5 5 5>
      constraintSetup {angularTargetAxis=AxisY, twistAxis=AxisZ, angularLimitsMax=<60 30 30>, angularLimitsMin=<-60 -30 -30>, angularTarget=<0 1 0>}
      localJointOffset=<0 0 0.2>
     }]
    relativeSpaceBone {boneName=thigh_r}
   }
   n6 = Anim Dynamics - Bone: BASE_dyn_jeans_l_04(ComponentPose=@n5, Alpha=1, ExternalForce="0,0,0") {
    angularDampingOverride=0.99
    angularSpringConstant=300
    bAngularSpring=true
    bLinearSpring=true
    bOverrideAngularDamping=true
    bOverrideLinearDamping=true
    bUseGravityOverride=true
    linearDampingOverride=0.999
    linearSpringConstant=60
    simulationSpace=World
    alphaBoolBlend {blendOption=Linear}
    boundBone {boneName=BASE_dyn_jeans_l_04}
    componentAppliedLinearAccClamp=<6000 6000 6000>
    componentLinearAccScale=<0.4 0.4 0.4>
    componentLinearVelScale=<0.4 0.4 0.4>
    physicsBodyDefinitions [{
      boundBone {boneName=BASE_dyn_jeans_l_04}
      boxExtents=<5 5 5>
      constraintSetup {angularTargetAxis=AxisY, twistAxis=AxisZ, angularLimitsMax=<60 30 30>, angularLimitsMin=<-60 -30 -30>, angularTarget=<0 1 0>}
      localJointOffset=<0 0 0.2>
     }]
    relativeSpaceBone {boneName=thigh_r}
   }
   n7 = Anim Dynamics - Bone: BASE_dyn_jeans_r_01(ComponentPose=@n6, Alpha=1, ExternalForce="0,0,0") {
    angularDampingOverride=0.99
    angularSpringConstant=300
    bAngularSpring=true
    bLinearSpring=true
    bOverrideAngularDamping=true
    bOverrideLinearDamping=true
    bUseGravityOverride=true
    linearDampingOverride=0.999
    linearSpringConstant=60
    simulationSpace=World
    alphaBoolBlend {blendOption=Linear}
    boundBone {boneName=BASE_dyn_jeans_r_01}
    componentAppliedLinearAccClamp=<6000 6000 6000>
    componentLinearAccScale=<0.4 0.4 0.4>
    componentLinearVelScale=<0.4 0.4 0.4>
    physicsBodyDefinitions [{
      boundBone {boneName=BASE_dyn_jeans_r_01}
      boxExtents=<5 5 5>
      constraintSetup {angularTargetAxis=AxisY, twistAxis=AxisZ, angularLimitsMax=<60 30 30>, angularLimitsMin=<-60 -30 -30>, angularTarget=<0 1 0>}
      localJointOffset=<0 0 0.2>
     }]
    relativeSpaceBone {boneName=thigh_r}
   }
   n8 = Anim Dynamics - Bone: BASE_dyn_jeans_r_02(ComponentPose=@n7, Alpha=1, ExternalForce="0,0,0") {
    angularDampingOverride=0.99
    angularSpringConstant=300
    bAngularSpring=true
    bLinearSpring=true
    bOverrideAngularDamping=true
    bOverrideLinearDamping=true
    bUseGravityOverride=true
    linearDampingOverride=0.999
    linearSpringConstant=60
    simulationSpace=World
    alphaBoolBlend {blendOption=Linear}
    boundBone {boneName=BASE_dyn_jeans_r_02}
    componentAppliedLinearAccClamp=<6000 6000 6000>
    componentLinearAccScale=<0.4 0.4 0.4>
    componentLinearVelScale=<0.4 0.4 0.4>
    physicsBodyDefinitions [{
      boundBone {boneName=BASE_dyn_jeans_r_02}
      boxExtents=<5 5 5>
      constraintSetup {angularTargetAxis=AxisY, twistAxis=AxisZ, angularLimitsMax=<60 30 30>, angularLimitsMin=<-60 -30 -30>, angularTarget=<0 1 0>}
      localJointOffset=<0 0 0.2>
     }]
    relativeSpaceBone {boneName=thigh_r}
   }
   n9 = Anim Dynamics - Bone: BASE_dyn_jeans_r_03(ComponentPose=@n8, Alpha=1, ExternalForce="0,0,0") {
    angularDampingOverride=0.99
    angularSpringConstant=300
    bAngularSpring=true
    bLinearSpring=true
    bOverrideAngularDamping=true
    bOverrideLinearDamping=true
    bUseGravityOverride=true
    linearDampingOverride=0.999
    linearSpringConstant=60
    simulationSpace=World
    alphaBoolBlend {blendOption=Linear}
    boundBone {boneName=BASE_dyn_jeans_r_03}
    componentAppliedLinearAccClamp=<6000 6000 6000>
    componentLinearAccScale=<0.4 0.4 0.4>
    componentLinearVelScale=<0.4 0.4 0.4>
    physicsBodyDefinitions [{
      boundBone {boneName=BASE_dyn_jeans_r_03}
      boxExtents=<5 5 5>
      constraintSetup {angularTargetAxis=AxisY, twistAxis=AxisZ, angularLimitsMax=<60 30 30>, angularLimitsMin=<-60 -30 -30>, angularTarget=<0 1 0>}
      localJointOffset=<0 0 0.2>
     }]
    relativeSpaceBone {boneName=thigh_r}
   }
   n10 = Anim Dynamics - Bone: BASE_dyn_jeans_r_04(ComponentPose=@n9, Alpha=1, ExternalForce="0,0,0") {
    angularDampingOverride=0.99
    angularSpringConstant=300
    bAngularSpring=true
    bLinearSpring=true
    bOverrideAngularDamping=true
    bOverrideLinearDamping=true
    bUseGravityOverride=true
    linearDampingOverride=0.999
    linearSpringConstant=60
    simulationSpace=World
    alphaBoolBlend {blendOption=Linear}
    boundBone {boneName=BASE_dyn_jeans_r_04}
    componentAppliedLinearAccClamp=<6000 6000 6000>
    componentLinearAccScale=<0.4 0.4 0.4>
    componentLinearVelScale=<0.4 0.4 0.4>
    physicsBodyDefinitions [{
      boundBone {boneName=BASE_dyn_jeans_r_04}
      boxExtents=<5 5 5>
      constraintSetup {angularTargetAxis=AxisY, twistAxis=AxisZ, angularLimitsMax=<60 30 30>, angularLimitsMin=<-60 -30 -30>, angularTarget=<0 1 0>}
      localJointOffset=<0 0 0.2>
     }]
    relativeSpaceBone {boneName=thigh_r}
   }
   Output Pose(Result=Component To Local(ComponentPose=@n10)) {layerGroup=None}
  ]}
 EventGraph {nodes [Event BlueprintUpdateAnimation()]}
]
variables []
